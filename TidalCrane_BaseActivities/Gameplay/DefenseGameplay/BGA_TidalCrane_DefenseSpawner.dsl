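; Defense-activity spawner actor that runs a Living World AI encounter for TidalCrane base activities.
; BeginPlay plays prep audio and removes spawn collision where cosmetics run, otherwise caches the child Living World volume once playlist data is ready.
; SpawnAI starts the encounter from a static point provider after a short delay, requests event generation and relays AI spawn and death events.
; AggroOnSupplyDrop aggros spawned AI onto a supply drop, except swarmers.
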(event EventBeginPlay
  (Networking|CanExecuteCosmeticEvents
    (:True
      (Components|DestroyComponent (Variables|Default|GetSpawnCollision))
      (CallFunction|PrepPhaseEffects))
    (:False
      (Variables|InternalData|SetEncounterInProgress false)
      (Class|FortAsyncActionCurrentPlaylistReady|PlaylistDataReadyAsync)
      (bind _asfort_athena_living_world_volume (Utilities|Casting|CastToFortAthenaLivingWorldVolume (Variables|/Script/Engine.ChildActorComponent|GetChildActor (Variables|Default|GetLivingWorldVolume_Spawning))))
      (Variables|InternalData|SetLivingWorldVolume _asfort_athena_living_world_volume))))

(event Custom|SpawnAI
  ; Try Spawn AI: ToDo - Track Velocity from Pawns Inside Volume, not just when entering
  (Variables|InternalData|SetEncounterInProgress true)
  (CallFunction|ProviderRegistered (Variables|InternalData|GetLivingWorldVolume)))

(event Custom|AggroOnSupplyDrop
  (params (SupplyDrop :type Athena%20Supply%20Drop%20Object%20Reference))
  ; Set aggro for actors to supply drop if used (doesn't work on swarmers)
  (for _array_element (Variables|/Script/LagerRuntime.FortAthenaLivingWorldEncounterInstance|GetSpawnedActors (Variables|InternalData|GetEncounterRef))
    (bind _asathena_aicontroller (Utilities|Casting|CastToAthenaAIController _array_element))
    (AI|AggroonActor _asathena_aicontroller SupplyDrop)))

(event Custom|ProviderRegistered
  (params (PointProviderVolume :type Fort%20Athena%20Living%20World%20Volume%20Object%20Reference))
  ; Start Encounter for AI
  (Utilities|FlowControl|Delay 0.2)
  (bind (_isvalid _returnvalue) (Default|StartActivityEncounterStaticPointProvider (Game|GetGameState) self (Utilities|ResolveSoftReference (Variables|Default|GetLagerEncounter)) (Variables|Default|GetStaticPointProvider)))
  (if _isvalid
    (bind _output_get (Variables|InternalData|SetEncounterRef _returnvalue))
    (Class|FortAthenaLivingWorldManager|RequestEventGeneration (Class|FortAthenaLivingWorldManager|GetLivingWorldManager))
    (EventDispatchers|BindEventtoOnActorSpawned _output_get (EventDispatchers|CreateEvent :!event Custom|OnAISpawned))
    (EventDispatchers|BindEventtoOnActorDied _output_get (EventDispatchers|CreateEvent :!event Custom|OnAIDied))
    (else
      (Fortnite|FortShippingLog self "Tried to Spawn Encounter for Mound, but Encounter failed to start." true))))

(event Custom|OnAISpawned
  (params (EncounterInstance :type Fort%20Athena%20Living%20World%20Encounter%20Instance%20Object%20Reference) (SpawnedActor :type Actor%20Object%20Reference) (EventData :type Fort%20Athena%20Living%20World%20Event%20Data%20Object%20Reference))
  (Default|CallSpawnAIEvent SpawnedActor))

(event Custom|OnAIDied
  (params (EncounterInstance :type Fort%20Athena%20Living%20World%20Encounter%20Instance%20Object%20Reference) (DiedActor :type Actor%20Object%20Reference))
  (Default|CallAIDiedEvent DiedActor))

(fn OnRep_EncounterInProgress)

(fn Prep_Phase_Effects
  (bind _returnvalue (Audio|SpawnSoundatLocation (Variables|Default|GetPrep%20Phase%20Sound) (Transformation|GetActorLocation)))
  (Variables|Default|SetPreppedAudio _returnvalue))

(fn ConstructionScript)
