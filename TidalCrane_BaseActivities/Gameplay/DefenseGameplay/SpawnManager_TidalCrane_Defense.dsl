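; Server-driven manager for the TidalCrane base defense event, where players kill spawned bugs at a base with friendly NPCs.
; BeginPlay sets a team-size-scaled target and, when enabled for testing, activates klaxons and spawners.
; Weighted kills (non-swarmers count triple) fill progress, polled by timer to grant chests and quest credit; storm consumption fails it.
; It also tracks overlapped NPCs, leashes bugs, broadcasts replicated state through GameplayEventRouter and submits participation analytics.
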(event Custom|AISpawned
  (params (SpawnedActor :type Actor%20Object%20Reference))
  ; On spawned for tracking the bugs
  ;;[0.01931 0.15 0.09414 0.5] AI Spawned
  (Default|UpdateAIActorLeash SpawnedActor)
  (Utilities|Array|Add (Variables|Encounter|GetActiveMinions) SpawnedActor) ; Keep active minions \ntracked for progress
  (CallFunction|UpdateLootTagforAI SpawnedActor))

(event Custom|Bug%20Died
  (params (DestroyedActor :type Actor%20Object%20Reference))
  (bind _eventdata (Variables|Encounter|GetEventData))
  (bind (_eventdata_spawnmanager_14_a13c79344bac9fd301fbf6be5b2811b6 _eventdata_eventstate_10_5aacbe6344f0c1f9283b6f996ee80287 _eventdata_alivenpcs_2_3c75bf2948d8cae5eae1b7a478022f31 _eventdata_maxnpcs_4_8ff348804ec5364cd593cfb7d51198f9 _eventdata_bugskilled_6_97a8a9e7411f76a1dadb4daa60527f22) (Utilities|Struct|BreakTidalCraneBaseDefenseData (Variables|GetEventData_SpawnManager_14_A13C79344BAC9FD301FBF6BE5B2811B6)))
  ; On spawned for tracking the bugs
  ;;[0.01931 0.15 0.09414 0.5] AI Spawned
  (Utilities|Array|RemoveItem (Variables|Encounter|GetActiveMinions) DestroyedActor)
  (bind _asfort_aipawn (Utilities|Casting|CastToFortAIPawn DestroyedActor))
  (bind (_outdeathinfo_finisherordowner _outdeathinfo_deathinstigator _outdeathinfo_deathcause _outdeathinfo_deathlocation _outdeathinfo_binitialized _outdeathinfo_deathtags _outdeathinfo_finisherordownertags _outdeathinfo_victimtags) (Utilities|Struct|BreakAIDeathInfo (Class|FortAIPawn|GetDeathInfo _asfort_aipawn)))
  (bind _asfort_player_controller_athena (Utilities|Casting|CastToFortPlayerControllerAthena _outdeathinfo_deathinstigator))
  ;;;[0.000356 0.04368 0.15 0.5] Participation Analytics
  (if (Utilities|IsValid (Variables|/Script/Engine.Controller|GetPlayerState _asfort_player_controller_athena) :!class /Script/Engine.KismetSystemLibrary)
    (Utilities|Set|Add (Variables|Analytics|GetParticipatingTeams) (Team|GetActorTeam (Variables|/Script/Engine.Controller|GetPlayerState _asfort_player_controller_athena)))
    (Utilities|Set|Add (Variables|Analytics|GetParticipatingPlayers) (Default|GetGameAccountId _asfort_player_controller_athena))
    (if (Utilities|String|IsEmpty (Variables|Analytics|GetFirstInteractingPlayer))
      (bind _returnvalue (PlayerInfo|GetDebugStringforUniqueId (Default|GetGameAccountId _asfort_player_controller_athena)))
      (Variables|Analytics|SetFirstInteractingPlayer _returnvalue)
      (Variables|Analytics|SetFirstInterationTime (Fortnite|Sync|Time|GetServerWorldRealTimeSeconds))
      (else
        _))
    (if (GameplayTags|HasTag _outdeathinfo_victimtags "(TagName=\"AISpawnerData.Type.TidalCrane.Swarmer\")" true)
      (bind _structout (Utilities|Struct|SetmembersinTidalCraneBaseDefenseData _eventdata (+ _eventdata_bugskilled_6_97a8a9e7411f76a1dadb4daa60527f22 1)))
      (Variables|Encounter|SetEventData _structout)
      (else
        (bind _structout_2 (Utilities|Struct|SetmembersinTidalCraneBaseDefenseData _eventdata (+ _eventdata_bugskilled_6_97a8a9e7411f76a1dadb4daa60527f22 3)))
        (Variables|Encounter|SetEventData _structout_2)))
    (else
      _)))

(event EventBeginPlay
  (|CallParentFunction)
  (bind (_gamestate _playlist) (Class|FortAsyncActionCurrentPlaylistReady|PlaylistDataReadyAsync)
    (:OnReady
      (bind _asfort_playlist_athena (Utilities|Casting|CastToFortPlaylistAthena _playlist))
      (Networking|SwitchHasAuthority
        (:Authority
          (bind _structout (Utilities|Struct|SetmembersinTidalCraneBaseDefenseData (Variables|Encounter|GetEventData) self "NewEnumerator0" (Utilities|Array|Length (Variables|Default|GetNPCs)) (Utilities|Array|Length (Variables|Default|GetNPCs)) (ScalableFloat|GetValueasInteger (Variables|Hotfixes|GetHFRow%20TargetBugsForEvent) (- (Math|Conversions|ToFloat(Integer) (Playlist|GetMaxTeamSize _gamestate _asfort_playlist_athena)) 1.0))))
          (Variables|Encounter|SetEventData _structout)
          (for _array_element (Variables|Default|GetNPCs)
            (EventDispatchers|BindEventtoOnDied _array_element (delegate Custom|NPCDied)))
          (CallFunction|SendGameplayEventtoNPCs "(TagName=\"Event.NPC.TidalCrane.BaseDefense.Preamble\")")
          (if (Variables|Default|GetEnabled%20For%20Testing)
            ; Register density for the spawners and begin the pre-event timer
            (Class|BTidalCraneBaseLocator|ReleaseDensity (Variables|Default|GetBaseLocator)) ; BaseLocator reserved density at start of match\nso room is guaranteed for this event.\nRelease reservation and take over management
            (for (_array_element_2 _array_index) (Variables|Default|GetBug%20Spawners)
              (Class|FortAthenaLivingWorldManager|RegisterDensityReservoir (Class|FortAthenaLivingWorldManager|GetLivingWorldManager) _array_element_2 (* (ScalableFloat|GetValueatLevel (Variables|AIBudget|GetAI%20Resevoir%20Density) (Math|Conversions|ToFloat(Integer) _array_index)) (Variables|/TidalCrane_BaseActivities/Gameplay/DefenseGameplay/BGA_TidalCrane_DefenseSpawner.BGA_TidalCrane_DefenseSpawner_C|GetDensity%20Mult _array_element_2)) (ScalableFloat|GetValueatLevel (Variables|AIBudget|GetAI%20Resevoir%20Range)))
              (Utilities|IsValid (Variables|/Script/Engine.ChildActorComponent|GetChildActor (Variables|/TidalCrane_BaseActivities/Gameplay/DefenseGameplay/BGA_TidalCrane_DefenseSpawner.BGA_TidalCrane_DefenseSpawner_C|GetLivingWorldVolume_Spawning _array_element_2))
                (:Is%20Valid
                  (bind _asfort_athena_living_world_volume (Utilities|Casting|CastToFortAthenaLivingWorldVolume (Variables|/Script/Engine.ChildActorComponent|GetChildActor (Variables|/TidalCrane_BaseActivities/Gameplay/DefenseGameplay/BGA_TidalCrane_DefenseSpawner.BGA_TidalCrane_DefenseSpawner_C|GetLivingWorldVolume_Spawning _array_element_2))))
                  (Class|FortAthenaLivingWorldPointProviderInterface|EnablePointProvider_DEPRECATED _asfort_athena_living_world_volume))))
            (CallFunction|ActivateKlaxons)
            (CallFunction|StartEvent)))))))

(event Custom|NPCDied
  (params (DamagedActor :type Actor%20Object%20Reference) (Damage :type Float%20%28single-precision%29) (InstigatedBy :type Controller%20Object%20Reference) (DamageCauser :type Actor%20Object%20Reference) (HitLocation :type Vector) (FHitComponent :type Primitive%20Component%20Object%20Reference) (BoneName :type Name) (Momentum :type Vector))
  (bind (_eventdata_spawnmanager_14_a13c79344bac9fd301fbf6be5b2811b6 _eventdata_eventstate_10_5aacbe6344f0c1f9283b6f996ee80287 _eventdata_alivenpcs_2_3c75bf2948d8cae5eae1b7a478022f31) (Utilities|Struct|BreakTidalCraneBaseDefenseData (Variables|GetEventData_SpawnManager_14_A13C79344BAC9FD301FBF6BE5B2811B6)))
  ; Remove an NPC from tracking (in the actor list, but keep their pawn data) when they die
  (bind _asfort_player_pawn (Utilities|Casting|CastToFortPlayerPawn DamagedActor))
  (Utilities|Array|RemoveItem (Variables|Default|GetNPCs) _asfort_player_pawn)
  (bind _structout (Utilities|Struct|SetmembersinTidalCraneBaseDefenseData (Variables|Encounter|GetEventData) (- _eventdata_alivenpcs_2_3c75bf2948d8cae5eae1b7a478022f31 1)))
  (Variables|Encounter|SetEventData _structout))

(event Custom|StartEvent
  ; Wait till it's time to attack
  (Networking|SwitchHasAuthority
    (:Authority
      (CallFunction|ActivateEvent)
      ;; Assuming using default variant spawners
      (for _array_element (Variables|Default|GetBug%20Spawners)
        (Default|BindEventtoAIDiedEvent _array_element (EventDispatchers|CreateEvent :!event Custom|Bug%20Died))
        (Default|BindEventtoSpawnAIEvent _array_element (EventDispatchers|CreateEvent :!event Custom|AISpawned))
        (Class|BGATidalCraneDefenseSpawner|SpawnAI _array_element))
      (bind _returnvalue (Utilities|Time|SetTimerbyEvent (EventDispatchers|CreateEvent :!event Custom|Event%20In%20Progress) 0.1 true))
      (Variables|Default|SetEvent%20Active%20Handle _returnvalue)
      (Utilities|Time|ClearandInvalidateTimerbyHandle (Variables|Default|GetSpawn%20Timer%20handle)))))

(event Custom|Event%20In%20Progress
  (bind (_spawnmanager_14_a13c79344bac9fd301fbf6be5b2811b6 _eventstate_10_5aacbe6344f0c1f9283b6f996ee80287 _alivenpcs_2_3c75bf2948d8cae5eae1b7a478022f31 _maxnpcs_4_8ff348804ec5364cd593cfb7d51198f9 _bugskilled_6_97a8a9e7411f76a1dadb4daa60527f22 _maxbugs_27_e3a9792f47777d86969e85a4a891a09f) (Utilities|Struct|BreakTidalCraneBaseDefenseData (Variables|Encounter|GetEventData)))
  ; Timer for while the event is active. Grant reward once this timer is complete if an NPC survives.
  (if (and (>= _bugskilled_6_97a8a9e7411f76a1dadb4daa60527f22 _maxbugs_27_e3a9792f47777d86969e85a4a891a09f) (not (or (Utilities|Enum|Equal(Enum) _eventstate_10_5aacbe6344f0c1f9283b6f996ee80287 "NewEnumerator5") (Utilities|Enum|Equal(Enum) _eventstate_10_5aacbe6344f0c1f9283b6f996ee80287 "NewEnumerator6"))))
    (CallFunction|KlaxonCleanup)
    (CallFunction|SendGameplayEventtoNPCs "(TagName=\"Event.NPC.TidalCrane.BaseDefense.Complete\")")
    (Default|SendCompletionQuestEvent "(GameplayTags=((TagName=\"Athena.Quests.BaseActivity.Defense\")))")
    (CallFunction|SpawnChestRewards)
    (CallFunction|EndEncounter true)))

(event OnComponentBeginOverlap(Box)
  (params (OverlappedComponent :type Primitive%20Component%20Object%20Reference) (OtherActor :type Actor%20Object%20Reference) (OtherComp :type Primitive%20Component%20Object%20Reference) (OtherBodyIndex :type Integer) (bFromSweep :type Boolean) (SweepResult :type Hit%20Result%20Structure%20%28by%20ref%29))
  (bind _npcs (Variables|Default|GetNPCs))
  ; Handle initial NPC overlap when DataLayer loads
  (Networking|SwitchHasAuthority
    (:Authority
      (bind (_bisnpc _npc) (CallFunction|ProcessNPC OtherActor))
      (if _bisnpc
        (if (not (Utilities|Array|ContainsItem _npcs _npc))
          (Utilities|Array|AddUnique _npcs _npc)
          (CallFunction|OnNPCAdded _npc))))))

(event Custom|OnNPCAdded
  (params (NPC :type Fort%20Player%20Pawn%20Object%20Reference))
  ; On NPC Added
  (CallFunction|ApplyorRemoveEffectsonNPC false NPC)
  (Utilities|FlowControl|Delay 0.2)
  (CallFunction|SendGameplayEventtoNPCs "(TagName=\"Event.NPC.TidalCrane.BaseDefense.Active\")"))

(event EventEndPlay
  (params (EndPlayReason :type EEndPlayReason%20Enum :default "Destroyed"))
  (Networking|SwitchHasAuthority
    (:Authority
      (CallFunction|OnSubmitAnalytics false) ; Submit Analytics (handles case where event didn't finish before match ended)
    )))

(fn ActivateKlaxons
  (for _array_element (Variables|Default|GetKlaxons)
    (bind _output_get (Variables|/HeistBasesGameplay/Items/Environmental/BP_Lockdown_Prop_Parent.BP_Lockdown_Prop_Parent_C|SetAlarmState true _array_element))
    (Class|BPLockdownPropParent|UpdateAlarmVisuals _array_element false _output_get))
  (return))

(fn ActivateEvent
  ; TODO: Remove the Timer data once the UI has been updated for the new design
  (bind _output_get (Variables|Encounter|SetProgressStartTime (Fortnite|Sync|Time|GetServerWorldTimeSeconds)))
  (bind _structout (Utilities|Struct|SetmembersinTidalCraneBaseDefenseData (Variables|Encounter|GetEventData) "NewEnumerator2" _output_get (ScalableFloat|GetValueatLevel (Variables|Default|GetEvent%20Duration))))
  (Variables|Encounter|SetEventData _structout)
  (CallFunction|SendGameplayEventtoNPCs "(TagName=\"Event.NPC.TidalCrane.BaseDefense.Active\")")
  (for _array_element (Variables|Default|GetBug%20Spawners)
    (Audio|Components|Audio|Stop (Variables|/TidalCrane_BaseActivities/Gameplay/DefenseGameplay/BGA_TidalCrane_DefenseSpawner.BGA_TidalCrane_DefenseSpawner_C|GetPreppedAudio _array_element))))

(fn EndEncounter
  (params (bSuccess :type Boolean) (bFailedByStorm :type Boolean))
  (bind _bsuccess (Variables|GetbSuccess))
  (bind _returnvalue (select _bsuccess "NewEnumerator5" "NewEnumerator6"))
  (bind _baselocator (Variables|Default|GetBaseLocator))
  (for _array_element (Variables|Default|GetBug%20Spawners)
    (Class|BPFLLagerCommunFunctions|EndActivityEncounter (Game|GetGameState) (Variables|/TidalCrane_BaseActivities/Gameplay/DefenseGameplay/BGA_TidalCrane_DefenseSpawner.BGA_TidalCrane_DefenseSpawner_C|GetEncounterRef _array_element)))
  (Utilities|Time|ClearandInvalidateTimerbyHandle (Variables|Default|GetEvent%20Active%20Handle))
  (Utilities|IsValid _baselocator
    (:Is%20Valid
      (Class|BTidalCraneBaseLocator|SetBaseMinimapIcon _baselocator _returnvalue)
      (bind _structout (Utilities|Struct|SetmembersinTidalCraneBaseDefenseData (Variables|Encounter|GetEventData) _returnvalue))
      (Variables|Encounter|SetEventData _structout)
      (CallFunction|OnSubmitAnalytics (Variables|GetbSuccess) (Variables|GetbFailedByStorm)))))

(fn GetAssociatedNPCs
  (return (Variables|Default|GetNPCs)))

(fn GetCountdownEventData
  (return (Utilities|Struct|MakeTidalCraneBaseTimerEvent self "NewEnumerator1" (Variables|Encounter|GetCountdownStartTime) (+ (Variables|Encounter|GetCountdownStartTime) (ScalableFloat|GetValueatLevel (Variables|Default|GetTime%20Until%20Event))) (ScalableFloat|GetValueatLevel (Variables|Default|GetTime%20Until%20Event)))))

(fn GetCountdownTimeRemaining
  (Networking|SwitchHasAuthority
    (:Authority
      (return (ScalableFloat|GetValueatLevel (Variables|Default|GetTime%20Until%20Event))))
    (:Remote
      (return (Math|Float|Max(Float) (- (ScalableFloat|GetValueatLevel (Variables|Default|GetTime%20Until%20Event)) (Math|Float|Max(Float) (- (Fortnite|Sync|Time|GetServerWorldTimeSeconds) (Variables|Encounter|GetCountdownStartTime)))))))))

(fn GetEventData
  (return (Variables|Encounter|GetEventData)))

(fn GetLeashLocation
  (bind _leashlocationactor (Variables|Encounter|GetLeashLocationActor)
    (:then
      (return (Transformation|GetActorLocation _leashlocationactor)))
    (:else
      (return (Transformation|GetActorLocation)))))

(fn GetProgressEventData
  (return (Utilities|Struct|MakeTidalCraneBaseTimerEvent self "NewEnumerator2" (Variables|Encounter|GetProgressStartTime) (+ (Variables|Encounter|GetProgressStartTime) (ScalableFloat|GetValueatLevel (Variables|Default|GetEvent%20Duration))) (ScalableFloat|GetValueatLevel (Variables|Default|GetEvent%20Duration)))))

(fn GetProgressTimeRemaining
  (Networking|SwitchHasAuthority
    (:Authority
      (return (ScalableFloat|GetValueatLevel (Variables|Default|GetEvent%20Duration))))
    (:Remote
      (return (Math|Float|Max(Float) (- (Variables|Default|GetEvent%20Duration) (Math|Float|Max(Float) (- (Fortnite|Sync|Time|GetServerWorldTimeSeconds) (Variables|Encounter|GetProgressStartTime)))))))))

(fn KlaxonCleanup
  (for _array_element (Variables|Default|GetKlaxons)
    (bind _output_get (Variables|/HeistBasesGameplay/Items/Environmental/BP_Lockdown_Prop_Parent.BP_Lockdown_Prop_Parent_C|SetAlarmState false _array_element))
    (Class|BPLockdownPropParent|UpdateAlarmVisuals _array_element true _output_get))
  (return))

(fn LeashNPCs
  (for _array_element (Variables|Default|GetNPCs)
    (CallFunction|SetNPCLeashtoReward _array_element)))

(fn OnConsumedByStorm
  (|CallParentFunction)
  (Networking|SwitchHasAuthority
    (:Authority
      (CallFunction|KlaxonCleanup)
      (CallFunction|EndEncounter false true)
      (for _array_element (Variables|Default|GetNPCs)
        (CallFunction|ApplyorRemoveEffectsonNPC true _array_element)))))

(fn OnPlayerPresenceChanged
  (|CallParentFunction)
  (for _array_element (Variables|Encounter|GetActiveMinions)
    (CallFunction|UpdateLootTagforAI _array_element)))

(fn OnRep_CountdownStartTime
  (bind _returnvalue (GameplayEventRouter|GetEventRouter "/Script/FortniteGame.FortPlayspace" self))
  (GameplayEventRouter|BroadcastEvent _returnvalue (Utilities|Struct|MakeTidalCraneBaseTimerEvent self "NewEnumerator1" (Variables|Encounter|GetCountdownStartTime) (+ (Variables|Encounter|GetCountdownStartTime) (ScalableFloat|GetValueatLevel (Variables|Default|GetTime%20Until%20Event))) (ScalableFloat|GetValueatLevel (Variables|Default|GetTime%20Until%20Event))) 0))

(fn OnRep_EventData
  (bind _returnvalue (GameplayEventRouter|GetEventRouter "/Script/FortniteGame.FortPlayspace" self))
  (GameplayEventRouter|BroadcastEvent _returnvalue (Variables|Encounter|GetEventData) 0))

(fn OnRep_ProgressStartTime
  (bind _returnvalue (GameplayEventRouter|GetEventRouter "/Script/FortniteGame.FortPlayspace" self))
  (GameplayEventRouter|BroadcastEvent _returnvalue (Utilities|Struct|MakeTidalCraneBaseTimerEvent self "NewEnumerator2" (Variables|Encounter|GetProgressStartTime) (+ (Variables|Encounter|GetProgressStartTime) (ScalableFloat|GetValueatLevel (Variables|Default|GetEvent%20Duration))) (ScalableFloat|GetValueatLevel (Variables|Default|GetEvent%20Duration))) 0))

(fn OnSubmitAnalytics
  (params (bSuceeded :type Boolean) (bCancelledForStorm :type Boolean))
  (bind (_returnvalue_x _returnvalue_y) (Math|Vector|BreakVector (Transformation|GetActorLocation self)))
  (bind _analyticseventattributes (LocalVariables|GetAnalyticsEventAttributes))
  (Networking|SwitchHasAuthority
    (:Authority
      (if (and (not (Replay|GetIsPlayingReplay)) (not (Variables|Analytics|GetbAnalyticsSubmitted)))
        (Variables|Analytics|SetbAnalyticsSubmitted true)
        (Utilities|FlowControl|Sequence
          (:then_0
            (Utilities|FlowControl|Sequence
              (:then_0
                (bind _returnvalue (Analytics|ConstructAnalyticsAttribute :AttributeName "Athena_Match_Time_Sec" :IntValue -1 :floatValue (Fortnite|Sync|Time|GetServerWorldRealTimeSeconds) :BoolValue false))
                (Utilities|Array|Add (LocalVariables|GetAnalyticsEventAttributes) _returnvalue))
              (:then_1
                (bind _returnvalue_2 (Analytics|ConstructAnalyticsAttribute "Location" (Utilities|String|Append (Utilities|String|ToString(Integer) (Math|Float|Truncate _returnvalue_x)) ", " (Utilities|String|ToString(Integer) (Math|Float|Truncate _returnvalue_y)))))
                (Utilities|Array|Add (LocalVariables|GetAnalyticsEventAttributes) _returnvalue_2))
              (:then_2
                (bind _returnvalue_3 (Analytics|ConstructAnalyticsAttribute "End_Reason" (Utilities|String|SelectString "Base defended successfully" (Utilities|String|SelectString "Base defeated (engulfed by storm)" "Base defeated (match ended)" (Variables|GetbCancelledForStorm)) (Variables|GetbSuceeded))))
                (Utilities|Array|Add (LocalVariables|GetAnalyticsEventAttributes) _returnvalue_3))
              (:then_3
                (bind _returnvalue_4 (Analytics|ConstructAnalyticsAttribute "Initiated_By" (Variables|Analytics|GetFirstInteractingPlayer)))
                (Utilities|Array|Add (LocalVariables|GetAnalyticsEventAttributes) _returnvalue_4))
              (:then_4
                (bind _returnvalue_5 (Analytics|ConstructAnalyticsAttribute :AttributeName "Initiated_Matchtime" :IntValue -1 :floatValue (Variables|Analytics|GetFirstInterationTime) :BoolValue false))
                (Utilities|Array|Add (LocalVariables|GetAnalyticsEventAttributes) _returnvalue_5))
              (:then_5
                (bind _returnvalue_6 (Analytics|ConstructAnalyticsAttribute :AttributeName "Bugs_Defeated" :IntValue (Utilities|Struct|BreakTidalCraneBaseDefenseData (Variables|Encounter|GetEventData)) :floatValue -1.0 :BoolValue false))
                (Utilities|Array|Add (LocalVariables|GetAnalyticsEventAttributes) _returnvalue_6))
              (:then_6
                (bind _returnvalue_7 (Analytics|ConstructAnalyticsAttribute :AttributeName "Unique_Players_Involved" :IntValue (Utilities|Set|Length (Variables|Analytics|GetParticipatingPlayers)) :floatValue -1.0 :BoolValue false))
                (Utilities|Array|Add (LocalVariables|GetAnalyticsEventAttributes) _returnvalue_7)
                (bind _returnvalue_8 (Analytics|ConstructAnalyticsAttribute :AttributeName "Unique_Teams_Involved" :IntValue (Utilities|Set|Length (Variables|Analytics|GetParticipatingTeams)) :floatValue -1.0 :BoolValue false))
                (Utilities|Array|Add (LocalVariables|GetAnalyticsEventAttributes) _returnvalue_8))))
          (:then_1
            (if (Utilities|Array|IsNotEmpty _analyticseventattributes)
              (Analytics|FireAnalyticsEventBlueprintArray "Core.TidalCraneBaseDefense.Complete" _analyticseventattributes))))))))

(fn SendCompletionQuestEvent
  (params (OptionalSubjectInfo :type Gameplay%20Tag%20Container%20Structure))
  (|CallParentFunction OptionalSubjectInfo))

(fn SendGameplayEventToNPCs
  (params (EventTag :type Gameplay%20Tag%20Structure))
  ;[0 0 0] Sends a gameplay event to NPCs to tell them to trigger certain barks, behaviors, etc.
  (bind _npcs (CallFunction|GetAssociatedNPCs))
  (for _array_element _npcs
    (Ability|SendGameplayEventtoActor _array_element EventTag)))

(fn StartEventCountdown
  (bind _output_get (Variables|Encounter|SetCountdownStartTime (Fortnite|Sync|Time|GetServerWorldTimeSeconds)))
  (bind _structout (Utilities|Struct|SetmembersinTidalCraneBaseDefenseData (Variables|Encounter|GetEventData) "NewEnumerator1" _output_get (ScalableFloat|GetValueatLevel (Variables|Default|GetTime%20Until%20Event))))
  (Variables|Encounter|SetEventData _structout)
  (CallFunction|ActivateKlaxons))

(fn UpdateAIActorLeash
  (params (AIActor :type Actor%20Object%20Reference))
  (bind _row_leashouterradius (Variables|Encounter|GetRow_LeashOuterRadius))
  (bind _returnvalue (ScalableFloat|GetValueatLevel _row_leashouterradius))
  (bind _row_leashinnerradius (Variables|Encounter|GetRow_LeashInnerRadius))
  (bind _returnvalue_2 (ScalableFloat|GetValueatLevel _row_leashinnerradius))
  (Utilities|IsValid AIActor
    (:Is%20Valid
      (if (and (> _returnvalue_2 0) (> _returnvalue 0))
        (bind _asfort_athena_aibot_controller (Utilities|Casting|CastToFortAthenaAIBotController (Fortnite|GetControllerfromActor AIActor)))
        (Class|BPFLAICommonFunctions|SetRadialLeash _asfort_athena_aibot_controller (Default|GetLeashLocation) _returnvalue_2 _returnvalue)))))

(fn ConstructionScript)
