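; Interactable switch Blueprint for the TidalCrane base activities test level, named as an activation switch in the defense gameplay folder.
; It adds no logic of its own: BlueprintOnInteract passes the interacting pawn and interaction type to the parent, and the construction script also calls its parent.
; BeginPlay, ActorBeginOverlap and Tick are empty stubs, so any switch behavior comes from outside this graph.
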(event EventBeginPlay)

(event Collision|EventActorBeginOverlap
  (params (OtherActor :type Actor%20Object%20Reference)))

(event EventTick
  (params (DeltaSeconds :type Float%20%28single-precision%29)))

(event Interaction|EventBlueprintOnInteract
  (params (InteractingPawn :type Fort%20Pawn%20Object%20Reference) (InteractionBeingAttempted :type EInteractionBeingAttempted%20Enum :default "FirstInteraction"))
  (|CallParentFunction InteractingPawn InteractionBeingAttempted))

(fn ConstructionScript
  (|CallParentFunction))
